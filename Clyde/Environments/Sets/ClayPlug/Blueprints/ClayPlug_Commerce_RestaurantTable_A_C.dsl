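; Restaurant table set-dressing Blueprint, judging by its name, that randomizes the rotation of its hierarchical instanced static mesh instances.
; The construction script loops over each instance and updates its local transform, keeping location and scale.
; It combines the current rotation with random yaw within plus or minus RotationAmount, plus the instance's own pitch and roll again.
; The randomness is unseeded and builds on current transforms; event graph stubs do nothing.
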(event EventBeginPlay
  (|CallParentFunction) ; This node is disabled and will not be called.\nDrag off pins to build functionality.
)

(event Collision|EventActorBeginOverlap
  (params (OtherActor :type Actor%20Object%20Reference)))

(event EventTick
  (params (DeltaSeconds :type Float%20%28single-precision%29)))

(fn ConstructionScript
  (|CallParentFunction)
  (for _index (range (Components|InstancedStaticMesh|GetInstanceCount (Variables|Default|GetHierarchicalInstancedStaticMesh)))
    (Components|InstancedStaticMesh|UpdateInstanceTransform (Variables|Default|GetHierarchicalInstancedStaticMesh) _index (Math|Transform|MakeTransform (.location (Components|InstancedStaticMesh|GetInstanceTransform (Variables|Default|GetHierarchicalInstancedStaticMesh) _index false)) (Math|Rotator|CombineRotators (.rotation (Components|InstancedStaticMesh|GetInstanceTransform (Variables|Default|GetHierarchicalInstancedStaticMesh) _index false)) (Math|Rotator|MakeRotator (.roll (.rotation (Components|InstancedStaticMesh|GetInstanceTransform (Variables|Default|GetHierarchicalInstancedStaticMesh) _index false))) (.pitch (.rotation (Components|InstancedStaticMesh|GetInstanceTransform (Variables|Default|GetHierarchicalInstancedStaticMesh) _index false))) (Math|Random|RandomFloatinRange (- (Variables|Default|GetRotationAmount) (* (Variables|Default|GetRotationAmount) 2.0)) (Variables|Default|GetRotationAmount)))) (.scale (Components|InstancedStaticMesh|GetInstanceTransform (Variables|Default|GetHierarchicalInstancedStaticMesh) _index false))) false false false)))
