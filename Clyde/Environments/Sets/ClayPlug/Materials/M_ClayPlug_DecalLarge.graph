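# Translucent deferred decal material for the ClayPlug set, with a default wet-floor mask and low roughness that suggest large wet patches.
# The mask samples tiled decal UVs, is scaled by Opacity and feeds a blend with constant grey, alongside a separate Roughness Value parameter.
# A world-aligned projection picks axes from rounded scene normals, with a camera-relative precision switch, but does not drive the mask sample.
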Material /Clyde/Environments/Sets/ClayPlug/Materials/M_ClayPlug_DecalLarge
graph {
 properties {
  bAutomaticallySetUsageInEditor=false
  bTangentSpaceNormal=false
  blendMode=BLEND_Translucent
  materialDomain=MD_DeferredDecal
  usedShadingModels=Default Lit
 }
 nodes [
  Constant() {r=1}
  n1 = StaticSwitchParameter(True=WorldPosition() {worldPositionShaderOffset=WPT_CameraRelative}.XYZ, False=WorldPosition().XYZ) {
   desc=Switch to Camera Relative World Position - Fix for floating point precision artifacts on large maps
   group="00) UV Mapping"
   parameterName=Large World Fix
   sortPriority=8
  }
  n2 = Multiply(A=@n1, B=Multiply(A=Constant() {r=0.002}))
  n3 = Round(None=Abs(None=SceneTexture() {sceneTextureId=PPI_WorldNormal}.Color))
  LinearInterpolate(
   A=LinearInterpolate(A=ComponentMask(None=@n2) {b=true, g=true}, B=ComponentMask(None=@n2) {b=true, r=true}, Alpha=ComponentMask(None=@n3) {g=true})
   B=ComponentMask(None=@n2) {g=true, r=true}
   Alpha=ComponentMask(None=@n3) {b=true}
  )
  n4 = Multiply(A=TextureCoordinate(), B=ScalarParameter() {defaultValue=3, parameterName=UV Tiling})
  LinearInterpolate(
   A=Constant3Vector() {constant=<rgba 0.5365 0.5365 0.5365 0>}
   B=Multiply(
    A=TextureSampleParameter2D(UVs=@n4) {parameterName=Decal Mask, texture=/Clyde/Environments/Sets/ClayPlug/Textures/T_Wet_Floor_Mask}.RGB
    B=ScalarParameter() {defaultValue=0.7, parameterName=Opacity}
   )
  )
  ScalarParameter() {defaultValue=0.1, parameterName=Roughness Value}
 ]
}
