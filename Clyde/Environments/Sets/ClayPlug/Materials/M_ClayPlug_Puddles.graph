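# Translucent deferred decal material for puddles in the ClayPlug environment set.
# A mask texture parameter, defaulting to a Fortnite effects decal texture, is multiplied by a scalar Opacity parameter.
# Color and Roughness are exposed as simple parameters, defaulting to a blue tint and a low, glossy roughness, with no normal map sampled.
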Material /Clyde/Environments/Sets/ClayPlug/Materials/M_ClayPlug_Puddles
graph {
 properties {
  bAutomaticallySetUsageInEditor=false
  bTangentSpaceNormal=false
  blendMode=BLEND_Translucent
  materialDomain=MD_DeferredDecal
  usedShadingModels=Default Lit
 }
 nodes [
  Multiply(
   A=ScalarParameter() {defaultValue=0.8, parameterName=Opacity}
   B=TextureSampleParameter2D() {parameterName=Mask, samplerType=SAMPLERTYPE_LinearColor, texture=/Game/Effects/Fort_Effects/Textures/Decal/T_Decal_Default_A}.RGB
  )
  ScalarParameter() {defaultValue=0.1, parameterName=Roughness}
  VectorParameter() {parameterName=Color, defaultValue=<rgba 0.0751 0.3445 0.8438 0>}
 ]
}
