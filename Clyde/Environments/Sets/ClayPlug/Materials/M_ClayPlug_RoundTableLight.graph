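# Translucent, unlit, two-sided material for the ClayPlug round table light, apparently a fake light-beam volume with a squared color through MF_EmissiveControls.
# World position offset uses rounded UV V to split the mesh ends, adjusting height in local space and flare along the tangent-space normal.
# Opacity combines vertical falloff, depth fade, separate front and backface fresnels, and near-camera fade.
# Two scrolling world-aligned smoke layers can modulate opacity, off by default.
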Material /Clyde/Environments/Sets/ClayPlug/Materials/M_ClayPlug_RoundTableLight
graph {
 properties {bAutomaticallySetUsageInEditor=false, blendMode=BLEND_Translucent, shadingModel=MSM_Unlit, twoSided=true, usedShadingModels=Unlit}
 nodes [
  n1 = Multiply(A=Constant3Vector() {constant=<rgba 0 0 1 0>}, B=ScalarParameter() {defaultValue=1, parameterName=WPO.Height})
  n2 = Round(None=ComponentMask(None=TextureCoordinate()) {g=true})
  n3 = Constant3Vector() {constant=<rgba 0 0 1 0>}
  n4 = Multiply(A=@n3, B=ScalarParameter() {parameterName=WPO.Top})
  n5 = Multiply(A=@n3, B=ScalarParameter() {defaultValue=10, parameterName=WPO.Bottom})
  Add(
   A=Transform(None=LinearInterpolate(A=@n1, Alpha=@n2) {constB=0}) {transformSourceType=TRANSFORMSOURCE_Local}
   B=Transform(None=LinearInterpolate(A=@n4, B=@n5, Alpha=@n2))
  )
  n6 = Power(Base=ComponentMask(None=TextureCoordinate()) {g=true}, Exp=ScalarParameter() {defaultValue=1, parameterName=Opacity.Verticle.FalloffPower})
  n7 = ScalarParameter() {defaultValue=1, parameterName=Opacity.Verticle.FalloffMulti}
  n8 = Multiply(A=ScalarParameter() {defaultValue=0.8, parameterName=Opacity}, B=Saturate(None=Multiply(A=@n6, B=@n7)))
  n9 = ScalarParameter() {defaultValue=100, parameterName=Opacity.DepthFade}
  n10 = DotProduct(A=CameraVectorWS(), B=VertexNormalWS())
  n11 = Add(A=ScalarParameter() {parameterName=BackfaceDotOffset}, B=Abs(None=@n10))
  n12 = Multiply(A=Saturate(None=@n11), B=ScalarParameter() {defaultValue=1, parameterName=Fresnel.Outer.Backface.Brightness})
  n13 = ScalarParameter() {defaultValue=1, parameterName=Fresnel.Outer.Backface.Exponent}
  n14 = Multiply(A=Saturate(None=OneMinus(None=@n11)), B=ScalarParameter() {defaultValue=1, parameterName=Fresnel.Inner.Backface.Brightness})
  n15 = ScalarParameter() {defaultValue=1, parameterName=Fresnel.Inner.Backface.Exponent}
  n16 = Multiply(A=Saturate(None=Power(Base=@n12, Exp=@n13)), B=Saturate(None=Power(Base=@n14, Exp=@n15)))
  n17 = ScalarParameter() {parameterName=FrontFaceDotOffset}
  n18 = Multiply(A=Saturate(None=Add(A=@n10, B=@n17)), B=ScalarParameter() {defaultValue=1, parameterName=Fresnel.Outer.Brightness})
  n19 = ScalarParameter() {defaultValue=1, parameterName=Fresnel.Outer.Exponent}
  n20 = Multiply(
   A=MaterialFunctionCall(
    Brightness=ScalarParameter() {defaultValue=1, parameterName=Fresnel.Inner.Brightness}
    Exponent=ScalarParameter() {defaultValue=1, parameterName=Fresnel.Inner.Exponent}
   ) {materialFunction=&MF_Fresnel_FX}.Fresnel Inverted
   B=Saturate(None=Power(Base=@n18, Exp=@n19))
  )
  n21 = Saturate(None=TwoSidedSign())
  n22 = Multiply(
   A=Multiply(A=DepthFade(Opacity=@n8, FadeDistance=@n9), B=LinearInterpolate(A=@n16, B=@n20, Alpha=@n21) {constB=0})
   B=MaterialFunctionCall(Distance=ScalarParameter() {defaultValue=250, parameterName=NearDistanceFade}) {materialFunction=&Near_Camera_Fade}.Result
  )
  n23 = TextureObject() {texture=/Game/Global/Textures/Default/Tileable_Patterns/T_Blowing_Smoke_Pattern2_Tile_DISP}
  n24 = ScalarParameter() {defaultValue=256, parameterName=Smoke.Main.Size}
  n25 = VectorParameter() {parameterName=Smoke.Main.Speed, defaultValue=<rgba 1 1 1 1>}
  n26 = Add(A=WorldPosition().XYZ, B=Multiply(A=Time(), B=Multiply(A=@n25.RGB, B=@n25.RGB)))
  n27 = ScalarParameter() {defaultValue=256, parameterName=Smoke.Secondary.Size}
  n28 = VectorParameter() {parameterName=Smoke.Secondary.Speed, defaultValue=<rgba 1 1 1 -1>}
  n29 = Add(A=WorldPosition().XYZ, B=Multiply(A=Time(), B=Multiply(A=@n28.RGB, B=@n28.RGB)))
  n30 = Add(
   A=LinearInterpolate(
    A=ScalarParameter() {parameterName=Smoke.Main.MinValue}
    Alpha=MaterialFunctionCall(TextureObject=@n23, TextureSize=@n24, WorldPosition=@n26) {materialFunction=&WorldAlignedTexture}.XYZ Texture
   )
   B=LinearInterpolate(
    A=ScalarParameter() {parameterName=Smoke.Secondary.MinValue}
    B=ScalarParameter() {defaultValue=1, parameterName=Smoke.Secondary.MaxValue}
    Alpha=MaterialFunctionCall(TextureObject=@n23, TextureSize=@n27, WorldPosition=@n29) {materialFunction=&WorldAlignedTexture}.XYZ Texture
   )
  )
  LinearInterpolate(A=@n22, B=Multiply(A=Saturate(None=@n30), B=@n22), Alpha=ScalarParameter() {parameterName=Smoke.MasterBlendAlpha, sliderMax=1})
  n31 = VectorParameter() {parameterName=EmissiveColor, channelNames {a=Intensity, b="", g="", r=""}, defaultValue=<rgba 0.9908 1 0.2958 1>}
  n32 = Multiply(A=@n31, B=@n31)
  MaterialFunctionCall!(BaseColor=@n32, EmissiveMask=@n32, EmissiveColor=@n32) {materialFunction=/Game/Athena/WinterArt/Materials/MF_EmissiveControls}
 ]
}
