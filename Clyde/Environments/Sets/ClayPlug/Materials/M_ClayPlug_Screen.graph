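# Prop material for the ClayPlug set that combines a textured base layer with an animated emissive screen, chosen by whether U exceeds 1.
# The screen layers a scalable image or tinted mask with Perlin fuzz, scanlines and a rolling band, using position-seeded timing, flicker and edge-fade terms, compensated for eye adaptation.
# Shared damage, wind, fan-spin, snow and move-tool functions follow, alongside backface normal correction and a cheaper low-quality mobile path.
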Material /Clyde/Environments/Sets/ClayPlug/Materials/M_ClayPlug_Screen
graph {
 properties {
  bAutomaticallySetUsageInEditor=false
  bUseMaterialAttributes=true
  bUsedWithInstancedStaticMeshes=true
  bUsedWithNanite=true
  bUsedWithSplineMeshes=true
  pixelDepthOffsetMode=PDOM_Legacy
  shadingModel=MSM_FromMaterialExpression
  usedShadingModels=Default Lit | Subsurface | Clear Coat | Two Sided Foliage
 }
 nodes [
  n1 = TextureCoordinate()
  n2 = TextureObjectParameter() {group=1_Layer, parameterName=Diffuse, sortPriority=0, texture=/Game/Global/Textures/Default/Blanks/T_MidgreysRGB}
  n3 = TextureObjectParameter() {
   desc="R = Specular\r\nG = Metal\r\nB = Roughness"
   group=1_Layer
   parameterName=SpecularMasks
   samplerType=SAMPLERTYPE_Masks
   sortPriority=0
   texture=&T_Fortnite_Default_S
  }
  n4 = TextureSample(Tex=@n3)
  n5 = TextureObjectParameter() {group=1_Layer, parameterName=Normals, samplerType=SAMPLERTYPE_Normal, sortPriority=0, texture=/Engine/EngineMaterials/FlatNormal}
  n6 = ScalarParameter() {defaultValue=1, group=2_Screen, parameterName=EmissiveMulti}
  n7 = MaterialFunctionCall(UVs=NamedRerouteUsage()) {materialFunction=&ScaleUVsByCenter}
  n8 = MaterialFunctionCall(X=Constant() {r=0.001}, Y=Constant() {r=2}) {materialFunction=&MakeFloat2}
  n9 = Time()
  n10 = MaterialFunctionCall(
   X=ScalarParameter() {defaultValue=2, group=2_Screen, parameterName=FuzzyScreen.Speed.X}
   Y=ScalarParameter() {defaultValue=0.1, group=2_Screen, parameterName=FuzzyScreen.Speed.Y}
  ) {materialFunction=&MakeFloat2}
  n11 = TextureSample(UVs=Add(A=Multiply(A=@n7.UVs, B=@n8.Result), B=Multiply(A=@n9, B=@n10.Result) {constB=0.04})) {samplerType=SAMPLERTYPE_LinearGrayscale, texture=&T_SimplePerlin_Greyscale}
  n12 = Constant()
  n13 = QualitySwitch(
   Default=Multiply(A=Power(Base=@n11.R), B=ScalarParameter() {defaultValue=1, group=2_Screen, parameterName=FuzzyScreen.Opacity})
   Low=@n12
   Medium=@n12
  )
  n14 = NamedRerouteUsage()
  n15 = TextureCoordinate()
  n16 = MaterialFunctionCall(X=Constant() {r=-1}) {materialFunction=&MakeFloat2}
  n17 = MaterialFunctionCall(
   UVs=NamedRerouteDeclaration(None=Add(A=@n15, B=@n16.Result)) {name=TexUVs, nodeColor=<rgba 1 0.1406 0 1>}
   Texture Scale=MaterialFunctionCall(
    X=ScalarParameter() {defaultValue=1, group=2_Screen, parameterName=ScreenImage.ScaleX}
    Y=ScalarParameter() {defaultValue=1, group=2_Screen, parameterName=ScreenImage.ScaleY}
   ) {materialFunction=&MakeFloat2}.Result
  ) {materialFunction=&ScaleUVsByCenter}
  n18 = TextureSampleParameter2D(UVs=@n17.UVs) {parameterName=ScreenImage, texture=&T_MidgreysRGB}
  n19 = LinearInterpolate(A=Constant() {r=1}, B=@n18.A, Alpha=ScalarParameter() {group=2_Screen, parameterName=ScreenImage.UseAlpha})
  n20 = VectorParameter() {group=2_Screen, parameterName=ImageAsMask.Tint, defaultValue=<rgba 0.8802 0.8802 0.8802 1>}
  n21 = ScalarParameter() {group=2_Screen, parameterName=ImageAsMask.MinValue}
  n22 = ScalarParameter() {defaultValue=1, group=2_Screen, parameterName=ImageAsMask.MaxValue}
  n23 = Desaturation(None=Multiply(A=VectorParameter() {group=2_Screen, parameterName=ImageAsMask.Channel, defaultValue=<rgba 1 1 1 1>}.RGB, B=@n18.RGB))
  n24 = LinearInterpolate(
   A=LinearInterpolate(
    A=VectorParameter() {group=2_Screen, parameterName=ScreenColor, defaultValue=<rgba 0 0 0 1>}.RGB
    B=@n18.RGB
    Alpha=Multiply(A=@n17.0-1 mask, B=@n19)
   )
   B=Multiply(A=@n20.RGB, B=LinearInterpolate(A=@n21, B=@n22, Alpha=@n23))
   Alpha=ScalarParameter() {group=2_Screen, parameterName=ImageAsMask.UseImageAsMask}
  )
  n25 = Multiply(A=@n20.RGB, B=NamedRerouteUsage())
  n26 = LinearInterpolate(
   A=Multiply(A=VectorParameter() {group=2_Screen, parameterName=ImageAsMask.Tint, defaultValue=<rgba 0.8802 0.8802 0.8802 1>}.RGB, B=Add(A=@n13, B=@n14))
   B=LinearInterpolate(A=@n24, B=Add(A=@n25, B=@n24), Alpha=NamedRerouteUsage())
   Alpha=@n18.A
  )
  n27 = Add(
   A=ComponentMask(None=TextureCoordinate()) {g=true}
   B=Multiply(A=Time(), B=ScalarParameter() {defaultValue=0.04, group=2_Screen, parameterName=Scanlines.AnimSpeed}) {constB=0.04}
  )
  n28 = ScalarParameter() {defaultValue=50, group=2_Screen, parameterName=Scanlines.Amount}
  n29 = ConstantBiasScale(None=Sine(None=Multiply(A=@n27, B=@n28) {constB=50}))
  n30 = ScalarParameter() {defaultValue=0.01, group=2_Screen, parameterName=Scanlines.Opacity}
  n31 = TextureCoordinate()
  n32 = MaterialFunctionCall(Float3=ActorPositionWS()) {materialFunction=&BreakOutFloat3Components}
  n33 = Frac(None=Add(A=Add(A=@n32.R, B=@n32.R), B=@n32.B))
  n34 = Add(A=Time(), B=Divide(A=NamedRerouteDeclaration(None=@n33) {name=Seed, nodeColor=<rgba 1 0.0234 0 1>}))
  n35 = Multiply(
   A=Multiply(A=@n34, B=ScalarParameter() {defaultValue=0.3, group=2_Screen, parameterName=ScreenEffect.Speed}) {constB=0.1}
   B=ScalarParameter() {defaultValue=0.3, group=2_Screen, parameterName=ScreenEffect.RandomOffsetSpeed}
  )
  n36 = Sine(None=@n35) {period=0.55}
  n37 = Sine(None=@n35) {period=0.33}
  n38 = Add(A=Add(A=ConstantBiasScale(None=@n36), B=ConstantBiasScale(None=@n37)), B=ConstantBiasScale(None=Sine(None=@n35) {period=64}))
  n39 = Add(
   A=Multiply(A=@n34, B=ScalarParameter() {defaultValue=0.3, group=2_Screen, parameterName=ScreenEffect.Speed}) {constB=0.1}
   B=Multiply(A=Add(A=@n38) {constB=2}, B=ScalarParameter() {defaultValue=0.2, group=2_Screen, parameterName=ScreenEffect.OffsetIntensity}) {constB=0.2}
  )
  n40 = ComponentMask(None=Frac(None=ActorPositionWS())) {g=true, r=true}
  n41 = TextureSample(UVs=Add(A=ComponentMask(None=@n31) {g=true}, B=Add(A=@n39, B=@n40))) {samplerType=SAMPLERTYPE_LinearGrayscale, texture=/Game/Global/Textures/Default/Masks_and_Gradients/T_ScreenRamp_H}
  n42 = Time()
  n43 = Sine(None=@n42)
  n44 = Sine(None=@n42) {period=0.45}
  n45 = Saturate(None=Add(A=Add(A=@n43, B=@n44), B=Sine(None=@n42) {period=1.63}))
  n46 = Power(Base=Multiply(A=@n41.R, B=LinearInterpolate(Alpha=@n45) {constA=0.4}))
  n47 = ScalarParameter() {defaultValue=1, group=2_Screen, parameterName=ScreenEffect.Opacity}
  n48 = NamedRerouteDeclaration(None=Add(A=Multiply(A=@n29, B=@n30) {constB=0.01}, B=Multiply(A=@n46, B=@n47) {constB=0.4})) {name=ScanlineOut, nodeColor=<rgba 0 1 0.3438 1>}
  n49 = Multiply(A=@n48, B=VectorParameter() {group=2_Screen, parameterName=ImageAsMask.Tint, defaultValue=<rgba 0.8802 0.8802 0.8802 1>}.RGB)
  n50 = ScalarParameter() {group=2_Screen, parameterName=ImageAsMask.UseImageAsMask}
  n51 = Multiply(A=NamedRerouteUsage(), B=Add(A=Multiply(A=@n26, B=@n6), B=LinearInterpolate(A=@n48, B=@n49, Alpha=@n50)))
  n52 = EyeAdaptation()
  n53 = BlendMaterialAttributes(
   A=MakeMaterialAttributes!(
    BaseColor=TextureSample(Tex=@n2).RGB
    Metallic=@n4.G
    Specular=@n4.G
    Roughness=@n4.G
    Anisotropy
    EmissiveColor
    Opacity
    OpacityMask
    Normal=TextureSample(Tex=@n5).RGB
    Tangent
    WorldPositionOffset
    SubsurfaceColor
    ClearCoat
    ClearCoatRoughness
    AmbientOcclusion
    Refraction
    CustomizedUVs_0
    CustomizedUVs_1
    CustomizedUVs_2
    CustomizedUVs_3
    CustomizedUVs_4
    CustomizedUVs_5
    CustomizedUVs_6
    CustomizedUVs_7
    PixelDepthOffset
    ShadingModel
    Displacement
   )
   B=MakeMaterialAttributes(BaseColor=@n26, Roughness=Constant() {r=0.7}, EmissiveColor=Divide(A=@n51, B=@n52.EyeAdaptation))
   Alpha=If(A=Constant() {r=1}, B=ComponentMask(None=@n1) {r=true}, "A > B"=Constant(), "A < B"=Constant() {r=1}) {desc=0}
  )
  MaterialExpressionReroute_69 = Reroute(None=@n53)
  n54 = MaterialFunctionCall!(Material Attributes=MaterialFunctionCall!(In=MaterialFunctionCall(Material_In=@MaterialExpressionReroute_69.MaterialAttributes) {materialFunction=&MF_FN_DestructibleDamage}.Material_Out) {materialFunction=/Game/Athena/WinterArt/Materials/MF_WPO_WindNew}.Result) {materialFunction=/Game/Creative/Sets/MilitaryBase/Props/MF_FanSpin}
  n55 = MaterialFunctionCall!(In=MaterialFunctionCall!(
    In=MaterialFunctionCall!(In=@n54.Result, UseSphericalMappingOverride) {materialFunction=/Game/Packages/Fortress_SharedMaterials/Base_Material_Functions/MF_CurieFallbackEffect}.Result
    Custom Mask
    Custom Rotation Axis
   ) {materialFunction=/Game/Athena/WinterArt/Materials/MF_WindWPO_v2}.Result) {materialFunction=/Game/Global/Base4Layers_2022/MaterialFunctions/MF_SnowAndIceArchitecture}
  MaterialExpressionReroute_68 = Reroute(None=@n55.Result)
  Composite!(None=@MaterialExpressionReroute_68.Result) {desc=Transform to World in case we will use RVT after this point, subgraphName=Tangent to World}
  Composite!(MaterialAttributes=@n53) {desc=Game Shared WPO and Fire Effect, subgraphName=WPO And Shared Functions}
  MaterialExpressionReroute_3 = Reroute(None=@MaterialExpressionReroute_68.Result)
  n56 = GetMaterialAttributes(None=@MaterialExpressionReroute_3.Input) {attributeGetTypes [#1]}
  n57 = Transform(None=@n56.Normal)
  n58 = LinearInterpolate(A=Multiply(A=@n57) {constB=-1}, B=@n57, Alpha=Saturate(None=TwoSidedSign())) {desc="Fixes normals\r\non backfaces"}
  n59 = SetMaterialAttributes!(
   MaterialAttributes=@n56.MaterialAttributes
   Normal=StaticSwitchParameter(True=MaterialProxyReplace(Realtime=@n58, MaterialProxy=@n56.Normal), False=@n58) {parameterName=Force TangentSpaceNormalsFor HLOD ProxyReplace}
  ) {attributeSetTypes [#1]}
  MaterialExpressionReroute_2 = Reroute(None=@n59)
  Composite!(InputPin=@MaterialExpressionReroute_2.Output) {subgraphName=Mobile Opt}
  MaterialExpressionReroute_7 = Reroute(None=@MaterialExpressionReroute_2.Output)
  n60 = GetMaterialAttributes(None=@MaterialExpressionReroute_7.InputPin)
  n61 = Constant() {r=0.5}
  n62 = Constant() {r=0.8}
  n63 = ShadingPathSwitch(
   Default=@MaterialExpressionReroute_7.InputPin
   Mobile=QualitySwitch(
    Default=@MaterialExpressionReroute_7.InputPin
    Low=SetMaterialAttributes!(MaterialAttributes=@n60.MaterialAttributes, Specular=@n61, Roughness=@n62) {attributeSetTypes [#2, #3]}
    High=@MaterialExpressionReroute_7.InputPin
    Medium=@MaterialExpressionReroute_7.InputPin
    Epic=@MaterialExpressionReroute_7.InputPin
   )
  )
  MaterialExpressionReroute_6 = Reroute(None=@n63)
  MaterialFunctionCall(MA In=@MaterialExpressionReroute_6.Output) {materialFunction=&MF_CreativeMoveTool}
  MaterialFunctionCall() {materialFunction=/Game/Global/Base4Layers_2022/MaterialFunctions/MF_GigaMix_2023}
  NamedRerouteDeclaration(None=Multiply(A=Multiply(A=@n41.R, B=@n45)) {constB=0.2}) {name=ScanlineBumpOffsetHeight, nodeColor=<rgba 0.6875 1 0 1>}
  n64 = MaterialFunctionCall(Float2=@n40) {materialFunction=&BreakOutFloat2Components}
  NamedRerouteDeclaration(None=Divide(A=Add(A=@n64.R, B=@n64.R))) {name=RandomAlpha, nodeColor=<rgba 1 0 0.5625 1>}
  n65 = NamedRerouteUsage()
  n66 = MaterialFunctionCall(Float2=@n65) {materialFunction=&BreakOutFloat2Components}
  n67 = Multiply(A=@n66.R, B=@n66.R)
  n68 = MaterialFunctionCall(Float2=OneMinus(None=@n65)) {materialFunction=&BreakOutFloat2Components}
  n69 = Multiply(A=@n68.R, B=@n68.R)
  n70 = Saturate(None=Multiply(A=Multiply(A=@n67, B=@n69)) {constB=40})
  n71 = Multiply(
   A=Power(Base=OneMinus(None=@n70), Exp=ScalarParameter() {defaultValue=16, group=2_Screen, parameterName=EdgeFade.Contrast}) {constExponent=16}
   B=ScalarParameter() {defaultValue=2, group=2_Screen, parameterName=EdgeFade.Multi}
  ) {constB=2}
  n72 = ScalarParameter() {defaultValue=0.2, group=2_Screen, parameterName=EdgeFade.Opacity}
  NamedRerouteDeclaration(None=Multiply(A=Saturate(None=@n71), B=@n72) {constB=0.1}) {name=EdgeMask, nodeColor=<rgba 0 1 0.1094 1>}
  n73 = Divide(A=@n71)
  NamedRerouteDeclaration(None=Multiply(A=Saturate(None=@n73), B=@n72) {constB=0.1}) {name=EdgeValues, nodeColor=<rgba 1 0.0234 0 1>}
  n74 = NamedRerouteUsage()
  n75 = Multiply(
   A=Add(A=Time(), B=Multiply(A=@n74) {constB=100})
   B=ScalarParameter() {defaultValue=2, group=2_Screen, parameterName=Brightness.FlickerSpeed}
  ) {constB=0.1}
  n76 = Sine(None=@n75) {period=0.57}
  n77 = Sine(None=@n75) {period=0.33}
  n78 = Add(A=Add(A=ConstantBiasScale(None=@n76), B=ConstantBiasScale(None=@n77)), B=ConstantBiasScale(None=Sine(None=@n75) {period=1.11}))
  NamedRerouteDeclaration(None=LinearInterpolate(A=ScalarParameter() {defaultValue=0.5, group=2_Screen, parameterName=Brightness.FlickerMin}, Alpha=Saturate(None=@n78)) {constA=0.5}) {name=BrightnessFlicker, nodeColor=<rgba 0 0.2969 1 1>}
  PinBase!(Output=@n59) {reroutePins [Output {expression=&self:MaterialExpressionReroute_2}]}
  PinBase!() {reroutePins [Input {expression=&self:MaterialExpressionReroute_3}]}
  PinBase() {reroutePins [MaterialAttributes {expression=&self:MaterialExpressionReroute_69}]}
  PinBase!(Result=@n55.Result) {reroutePins [Result {expression=&self:MaterialExpressionReroute_68}]}
  PinBase!() {reroutePins [InputPin {expression=&self:MaterialExpressionReroute_7}]}
  PinBase!(Output=@n63) {reroutePins [Output {expression=&self:MaterialExpressionReroute_6}]}
  Round(None=@n71)
 ]
}
