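# Translucent deferred decal material for wall graffiti in the ClayPlug environment set, with a swappable Graffiti texture defaulting to a Cooties image.
# The texture's alpha is scaled by an Alpha parameter, and Specular is a scalar parameter.
# A static switch selects either flat Roughness or a min-max blend driven by a tiled, channel-selectable mask.
# A separate branch turns a brick texture's red channel into a stepped, inverted mask through power, multiply and floor.
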Material /Clyde/Environments/Sets/ClayPlug/Materials/M_Clyde_WallGraffiti_Decal
graph {
 properties {blendMode=BLEND_Translucent, materialDomain=MD_DeferredDecal, pixelDepthOffsetMode=PDOM_Legacy, usedShadingModels=Default Lit}
 nodes [
  Constant()
  Multiply(
   A=TextureSampleParameter2D() {parameterName=Graffiti, texture=/Game/Environments/Sets/Graffiti/Textures/T_Graffiti_Cooties}.A
   B=ScalarParameter() {defaultValue=1, parameterName=Alpha}
  )
  n1 = Power(Base=TextureSample(UVs=TextureCoordinate() {uTiling=1.5}) {texture=/Game/Packages/AB_Fortnite/Textures/T_BrickWall_04_D}.R) {constExponent=1.2}
  OneMinus(None=Floor(None=Multiply(A=@n1) {constB=7}))
  ScalarParameter() {defaultValue=0.5, parameterName=Specular}
  n2 = TextureSampleParameter2D(UVs=Multiply(A=TextureCoordinate(), B=ScalarParameter() {defaultValue=1, parameterName=RoughnessMaskTiling})) {parameterName=RoughnessMask, samplerType=SAMPLERTYPE_Masks, texture=/Clyde/Packages/Fortress_SharedMaterials/Textures/WhiteTextureMask}
  StaticSwitchParameter(
   True=LinearInterpolate(
    A=ScalarParameter() {defaultValue=0.25, parameterName=RoughnessMin}
    B=ScalarParameter() {defaultValue=0.5, parameterName=RoughnessMax}
    Alpha=StaticComponentMaskParameter(None=@n2.R) {defaultR=true, parameterName=RoughnessMaskChannel}
   )
   False=ScalarParameter() {defaultValue=0.5, parameterName=Roughness}
  ) {parameterName=UseRoughnessMask}
 ]
}
