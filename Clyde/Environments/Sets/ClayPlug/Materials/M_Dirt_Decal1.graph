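# M_Dirt_Decal1 is a translucent deferred decal material for dirt marks in the ClayPlug environment set.
# It multiplies an exponential radial gradient, with adjustable radius and density, by the inverted dirt mask texture to fade the mark toward its edges.
# An opacity multiplier parameter scales that result, and a separate Color parameter supplies a flat brownish tint.
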Material /Clyde/Environments/Sets/ClayPlug/Materials/M_Dirt_Decal1
graph {
 properties {
  bCastRayTracedShadows=false
  blendMode=BLEND_Translucent
  materialDomain=MD_DeferredDecal
  pixelDepthOffsetMode=PDOM_Legacy
  twoSided=true
  usedShadingModels=Default Lit
 }
 nodes [
  n1 = ScalarParameter() {defaultValue=0.5, parameterName=Radius}
  n2 = ScalarParameter() {defaultValue=6, parameterName=Density}
  n3 = TextureSample() {texture=/Clyde/Environments/Sets/ClayPlug/Textures/T_Dirt_Decal_M}
  Multiply(
   A=Multiply(
    A=MaterialFunctionCall(Radius=@n1, Density=@n2) {materialFunction=&RadialGradientExponential}.RadialGradientExponential
    B=OneMinus(None=@n3.RGBA)
   )
   B=ScalarParameter() {defaultValue=0.5, parameterName=OpacityMutliplier}
  )
  VectorParameter() {parameterName=Color, defaultValue=<rgba 0.5747 0.3664 0.2863 1>}
 ]
}
